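# Physics asset for the second Cyber Karate male medium body cosmetic, covering its dangling skirt, belt straps, sleeves and pant legs.
# Box bodies sit on the Dyn bones, alongside pelvis and thigh capsules and calf and clavicle boxes.
# Six skirt panels hang from a master body with free-swinging lower segments and some neighbor links.
# Belt straps form two three-link chains, sleeves are angularly locked to clavicles, and pant legs pivot narrowly on calves.
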PhysicsAsset /BRCosmetics/Characters/Player/Male/Medium/Bodies/M_MED_Cyber_Karate/Meshes/M_MED_Cyber_Karate_02_PhysicsAsset
bodies {
 Dyn_beltStrap01_l [{shapeType=Box, extentX=4.5095, extentY=1, extentZ=8.2999, center=<4.262 0 0>, rotation=<rot 90 0 0>}]
 Dyn_beltStrap01_r [{shapeType=Box, extentX=4.5095, extentY=1, extentZ=9.1761, center=<5.0508 -0.0001 0>, rotation=<rot 90 0 0>}]
 Dyn_beltStrap02_l [{shapeType=Box, extentX=4.5095, extentY=1, extentZ=8.3424, center=<4.3831 0 0>, rotation=<rot 90 0 0>}]
 Dyn_beltStrap02_r [{shapeType=Box, extentX=4.5095, extentY=1, extentZ=9.3547, center=<4.4338 0 0>, rotation=<rot 90 0 0>}]
 Dyn_beltStrap03_l [{shapeType=Box, extentX=4.5095, extentY=1, extentZ=7.9653, center=<4 0 0>, rotation=<rot 90 0 0>}]
 Dyn_beltStrap03_r [{shapeType=Box, extentX=4.8832, extentY=1, extentZ=7.8506, center=<4 0 0>, rotation=<rot 90 0 0>}]
 Dyn_belt_mstr [{shapeType=Box, extentX=3, extentY=3, extentZ=3}]
 Dyn_pant_leg_l [{shapeType=Box, extentX=24.0543, extentY=20.396, extentZ=3.4384, center=<2.5327 0 0>, rotation=<rot 90 0 0>}]
 Dyn_pant_leg_r [{shapeType=Box, extentX=24.8073, extentY=19.2767, extentZ=3.4384, center=<2.5715 0 0>, rotation=<rot 90 0 0>}]
 Dyn_skirt_a [{shapeType=Box, extentX=13, extentY=3, extentZ=7, center=<4 0 0>, rotation=<rot 90 0 0>}]
 Dyn_skirt_a_02 [{shapeType=Box, extentX=13, extentY=3, extentZ=7, center=<4 0 0>, rotation=<rot 90 0 0>}]
 Dyn_skirt_b [{shapeType=Box, extentX=13, extentY=3, extentZ=7, center=<4 0 0>, rotation=<rot 90 0 0>}]
 Dyn_skirt_b_02 [{shapeType=Box, extentX=13, extentY=3, extentZ=7, center=<4 0 0>, rotation=<rot 90 0 0>}]
 Dyn_skirt_c [{shapeType=Box, extentX=13, extentY=3, extentZ=7, center=<4 0 0>, rotation=<rot 90 0 0>}]
 Dyn_skirt_c_02 [{shapeType=Box, extentX=13, extentY=3, extentZ=7, center=<4 0 0>, rotation=<rot 90 0 0>}]
 Dyn_skirt_d [{shapeType=Box, extentX=13, extentY=3, extentZ=7, center=<4 0 0>, rotation=<rot 90 0 0>}]
 Dyn_skirt_d_02 [{shapeType=Box, extentX=13, extentY=3, extentZ=7, center=<4 0 0>, rotation=<rot 90 0 0>}]
 Dyn_skirt_e [{shapeType=Box, extentX=15.5, extentY=3, extentZ=7, center=<4 0 0>, rotation=<rot 90 0 0>}]
 Dyn_skirt_e_02 [{shapeType=Box, extentX=16, extentY=3, extentZ=7, center=<4 0 0>, rotation=<rot 90 0 0>}]
 Dyn_skirt_f [{shapeType=Box, extentX=13, extentY=3, extentZ=7, center=<4 0 0>, rotation=<rot 90 0 0>}]
 Dyn_skirt_f_02 [{shapeType=Box, extentX=13, extentY=3, extentZ=7, center=<4 0 0>, rotation=<rot 90 0 0>}]
 Dyn_skirt_mstr [{shapeType=Box, extentX=6.6715, extentY=14.6294, extentZ=18.1194}]
 Dyn_sleeve_l [{shapeType=Box, extentX=5.676, extentY=7.9676, extentZ=13.1011, rotation=<rot 90 0 0>}]
 Dyn_sleeve_r [{shapeType=Box, extentX=4.5111, extentY=7.0112, extentZ=13.851, rotation=<rot 90 0 0>}]
 calf_l [{shapeType=Box, extentX=7.5, extentY=7.5, extentZ=7.5}]
 calf_r [{shapeType=Box, extentX=7.5, extentY=7.5, extentZ=7.5}]
 clavicle_l [{shapeType=Box, extentX=4.8061, extentY=10.4407, extentZ=10.8052, center=<11.3117 -0.6081 -2.0702>, rotation=<rot 0 -171.1823 -67.6902>}]
 clavicle_r [{shapeType=Box, extentX=6.6853, extentY=9.4038, extentZ=10.698, center=<-10.1372 0.1976 2.1166>, rotation=<rot 0 -173.8363 -77.2904>}]
 pelvis [{shapeType=Capsule, length=8.4294, radius=9.2606, center=<-1.1326 -3.4204 0.0001>}]
 thigh_twist_01_l [{shapeType=Capsule, length=13.2535, radius=9.7248, center=<10.1896 -0.1803 -0.9356>, rotation=<rot 90 0 0>}]
 thigh_twist_01_r [{shapeType=Capsule, length=13.2535, radius=9.7248, center=<-10.4624 0.1834 0>, rotation=<rot 90 0 0>}]
}
constraints [
 {bone1Name=Dyn_skirt_e, bone2Name=Dyn_skirt_mstr, swing1LimitDegrees=45, swing2LimitDegrees=45, twistLimitDegrees=45}
 {bone1Name=Dyn_skirt_d, bone2Name=Dyn_skirt_mstr, swing1LimitDegrees=45, swing2LimitDegrees=45, twistLimitDegrees=45}
 {bone1Name=Dyn_skirt_c, bone2Name=Dyn_skirt_mstr, swing1LimitDegrees=45, swing2LimitDegrees=45, twistLimitDegrees=45}
 {bone1Name=Dyn_skirt_b, bone2Name=Dyn_skirt_mstr, swing1LimitDegrees=45, swing2LimitDegrees=45, twistLimitDegrees=45}
 {bone1Name=Dyn_skirt_a, bone2Name=Dyn_skirt_mstr, swing1LimitDegrees=45, swing2LimitDegrees=45, twistLimitDegrees=28.8571}
 {bone1Name=Dyn_skirt_f, bone2Name=Dyn_skirt_mstr, swing1LimitDegrees=34.2857, swing2LimitDegrees=10.7143, twistLimitDegrees=10}
 {bone1Name=Dyn_beltStrap01_l, bone2Name=Dyn_belt_mstr, swing1LimitDegrees=75, swing2LimitDegrees=75, twistLimitDegrees=45}
 {bone1Name=Dyn_beltStrap02_l, bone2Name=Dyn_beltStrap01_l, swing1LimitDegrees=75, swing2LimitDegrees=75, twistLimitDegrees=45}
 {bone1Name=Dyn_beltStrap03_l, bone2Name=Dyn_beltStrap02_l, swing1LimitDegrees=75, swing2LimitDegrees=75, twistLimitDegrees=45}
 {bone1Name=Dyn_beltStrap01_r, bone2Name=Dyn_belt_mstr, swing1LimitDegrees=75, swing2LimitDegrees=75, twistLimitDegrees=45}
 {bone1Name=Dyn_beltStrap02_r, bone2Name=Dyn_beltStrap01_r, swing1LimitDegrees=75, swing2LimitDegrees=75, twistLimitDegrees=45}
 {bone1Name=Dyn_beltStrap03_r, bone2Name=Dyn_beltStrap02_r, swing1LimitDegrees=75, swing2LimitDegrees=75, twistLimitDegrees=45}
 {
  bone1Name=clavicle_l
  bone2Name=pelvis
  swing1LimitDegrees=45
  swing1Motion=Free
  swing2LimitDegrees=45
  swing2Motion=Free
  twistLimitDegrees=45
  twistMotion=Free
 }
 {
  bone1Name=clavicle_r
  bone2Name=pelvis
  swing1LimitDegrees=45
  swing1Motion=Free
  swing2LimitDegrees=45
  swing2Motion=Free
  twistLimitDegrees=45
  twistMotion=Free
 }
 {
  bone1Name=Dyn_sleeve_r
  bone2Name=clavicle_r
  swing1LimitDegrees=45
  swing1Motion=Locked
  swing2LimitDegrees=45
  swing2Motion=Locked
  twistLimitDegrees=45
  twistMotion=Locked
 }
 {
  bone1Name=Dyn_sleeve_l
  bone2Name=clavicle_l
  swing1LimitDegrees=45
  swing1Motion=Locked
  swing2LimitDegrees=45
  swing2Motion=Locked
  twistLimitDegrees=45
  twistMotion=Locked
 }
 {bone1Name=Dyn_pant_leg_r, bone2Name=calf_r, swing1LimitDegrees=15, swing2LimitDegrees=15, twistLimitDegrees=15}
 {bone1Name=Dyn_pant_leg_l, bone2Name=calf_l, swing1LimitDegrees=15, swing2LimitDegrees=15, twistLimitDegrees=15}
 {
  bone1Name=Dyn_skirt_a
  bone2Name=Dyn_skirt_b
  swing1LimitDegrees=45
  swing1Motion=Free
  swing2LimitDegrees=45
  swing2Motion=Free
  twistLimitDegrees=45
  twistMotion=Free
 }
 {
  bone1Name=Dyn_skirt_a
  bone2Name=Dyn_skirt_f
  swing1LimitDegrees=45
  swing1Motion=Free
  swing2LimitDegrees=45
  swing2Motion=Free
  twistLimitDegrees=45
  twistMotion=Free
 }
 {
  bone1Name=Dyn_skirt_f
  bone2Name=Dyn_skirt_a
  swing1LimitDegrees=45
  swing1Motion=Free
  swing2LimitDegrees=45
  swing2Motion=Free
  twistLimitDegrees=45
  twistMotion=Free
 }
 {
  bone1Name=Dyn_skirt_f_02
  bone2Name=Dyn_skirt_f
  swing1LimitDegrees=45
  swing1Motion=Free
  swing2LimitDegrees=45
  swing2Motion=Free
  twistLimitDegrees=45
  twistMotion=Free
 }
 {
  bone1Name=Dyn_skirt_a_02
  bone2Name=Dyn_skirt_a
  swing1LimitDegrees=45
  swing1Motion=Free
  swing2LimitDegrees=45
  swing2Motion=Free
  twistLimitDegrees=45
  twistMotion=Free
 }
 {
  bone1Name=Dyn_skirt_b_02
  bone2Name=Dyn_skirt_b
  swing1LimitDegrees=45
  swing1Motion=Free
  swing2LimitDegrees=45
  swing2Motion=Free
  twistLimitDegrees=45
  twistMotion=Free
 }
 {
  bone1Name=Dyn_skirt_c_02
  bone2Name=Dyn_skirt_c
  swing1LimitDegrees=45
  swing1Motion=Free
  swing2LimitDegrees=45
  swing2Motion=Free
  twistLimitDegrees=45
  twistMotion=Free
 }
 {
  bone1Name=Dyn_skirt_d_02
  bone2Name=Dyn_skirt_d
  swing1LimitDegrees=45
  swing1Motion=Free
  swing2LimitDegrees=45
  swing2Motion=Free
  twistLimitDegrees=45
  twistMotion=Free
 }
 {
  bone1Name=Dyn_skirt_e_02
  bone2Name=Dyn_skirt_e
  swing1LimitDegrees=45
  swing1Motion=Free
  swing2LimitDegrees=45
  swing2Motion=Free
  twistLimitDegrees=45
  twistMotion=Free
 }
]
