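# Crouch physics asset for the male medium Cyber Karate body, covering its dangling belt straps, skirt panels, sleeves and pant legs.
# Box bodies on Dyn_ bones form limited-angle chains anchored to master, calf and clavicle bones, alongside pelvis, thigh and upper-arm capsules.
# Skirt panels A and F are cross-linked at both segments.
# Sleeves are rotationally locked to the clavicles, while sleeve-adjust bodies may only twist.
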PhysicsAsset /BRCosmetics/Characters/Player/Male/Medium/Bodies/M_MED_Cyber_Karate/Meshes/M_MED_Cyber_Karate_CrouchPhysics
bodies {
 Dyn_beltStrap01_l [{shapeType=Box, extentX=3.6421, extentY=1.5, extentZ=9.1005, center=<4.7542 0 0>, rotation=<rot 90 0 0>}]
 Dyn_beltStrap01_l_mstr [{shapeType=Box, extentX=7.5, extentY=7.5, extentZ=7.5}]
 Dyn_beltStrap01_r [{shapeType=Box, extentX=3.7151, extentY=1.5177, extentZ=11.1031, center=<4.9393 -0.0002 0>, rotation=<rot 90 0 0>}]
 Dyn_beltStrap02_l [{shapeType=Box, extentX=4.9042, extentY=1.5, extentZ=8.6419, center=<4.0224 -0.2567 0.0411>, rotation=<rot 87.1798 -179.9988 -177.8262>}]
 Dyn_beltStrap02_r [{shapeType=Box, extentX=4.4741, extentY=1.7873, extentZ=11.1466, center=<5.5487 -0.2443 0.0224>, rotation=<rot 87.1987 179.9986 -176.8261>}]
 Dyn_beltStrap03_l [{shapeType=Box, extentX=5.2477, extentY=1.5, extentZ=7.5254, center=<3.4478 -0.0001 0>, rotation=<rot 90 0 0>}]
 Dyn_beltStrap03_r [{shapeType=Box, extentX=5.8256, extentY=2.2264, extentZ=7.1824, center=<4.327 -0.7343 0>, rotation=<rot 84.6344 -95.4216 -90.0005>}]
 Dyn_belt_mstr [{shapeType=Box, extentX=3, extentY=3, extentZ=3}]
 Dyn_pant_leg_l [{shapeType=Box, extentX=24.0543, extentY=20.396, extentZ=3.4384, center=<2.5327 0 0>, rotation=<rot 90 0 0>}]
 Dyn_pant_leg_r [{shapeType=Box, extentX=24.8073, extentY=19.2767, extentZ=3.4384, center=<2.5715 0 0>, rotation=<rot 90 0 0>}]
 Dyn_skirt_a [{shapeType=Box, extentX=17.6358, extentY=2, extentZ=8.5801, center=<3.3375 0 0.7286>, rotation=<rot 90 0 0>}]
 Dyn_skirt_a_02 [{shapeType=Box, extentX=18.1931, extentY=1.8078, extentZ=10.3437, center=<4.3438 0.8854 0.8693>, rotation=<rot 90 0 1.7977>}]
 Dyn_skirt_a_mstr [{shapeType=Box, extentX=7.5, extentY=7.5, extentZ=7.5}]
 Dyn_skirt_b [{shapeType=Box, extentX=13.5, extentY=2.9355, extentZ=7, center=<4.3273 0 0>, rotation=<rot 90 0 0>}]
 Dyn_skirt_b_02 [{shapeType=Box, extentX=14.0136, extentY=1.9649, extentZ=7.1131, center=<4.5267 0 0>, rotation=<rot 90 0 0>}]
 Dyn_skirt_b_mstr [{shapeType=Box, extentX=7.5, extentY=7.5, extentZ=7.5}]
 Dyn_skirt_back [{shapeType=Box, extentX=15.5, extentY=3, extentZ=7, center=<4 0 0>, rotation=<rot 90 0 0>}]
 Dyn_skirt_back_02 [{shapeType=Box, extentX=15.5, extentY=3, extentZ=7, center=<4 0 0>, rotation=<rot 90 0 0>}]
 Dyn_skirt_c [{shapeType=Box, extentX=15.5, extentY=3, extentZ=7, center=<4 0 0>, rotation=<rot 90 0 0>}]
 Dyn_skirt_c_02 [{shapeType=Box, extentX=15.5, extentY=3, extentZ=7, center=<4 0 0>, rotation=<rot 90 0 0>}]
 Dyn_skirt_d [{shapeType=Box, extentX=18, extentY=3.5, extentZ=6, center=<2.7369 0.0001 0>, rotation=<rot 90 0 0>}]
 Dyn_skirt_d_02 [{shapeType=Box, extentX=18, extentY=3, extentZ=6.5, center=<4 0 0>, rotation=<rot 90 0 0>}]
 Dyn_skirt_e [{shapeType=Box, extentX=21, extentY=6, extentZ=7, center=<4 0 0>, rotation=<rot 90 0 0>}]
 Dyn_skirt_e_02 [{shapeType=Box, extentX=21, extentY=5.5, extentZ=7, center=<4 0 0>, rotation=<rot 90 0 0>}]
 Dyn_skirt_f [{shapeType=Box, extentX=13.0573, extentY=1.2, extentZ=10.4675, center=<4.4343 0 0>, rotation=<rot 80.0771 -89.9994 -89.9994>}]
 Dyn_skirt_f_02 [{shapeType=Box, extentX=14.1825, extentY=0.8, extentZ=9.1085, center=<4.18 0 0>, rotation=<rot 79.2533 -89.9997 -89.9997>}]
 Dyn_skirt_f_mstr [{shapeType=Box, extentX=7.5, extentY=7.5, extentZ=7.5}]
 Dyn_skirt_mstr [{shapeType=Box, extentX=6.6715, extentY=14.6294, extentZ=18.1194}]
 Dyn_sleeve_adjust_l [{shapeType=Box, extentX=30.5688, extentY=5.9, extentZ=15.4725, center=<0.3276 0 8.0449>}]
 Dyn_sleeve_adjust_r [{shapeType=Box, extentX=30.8736, extentY=7.4612, extentZ=16.0393, center=<0.13 0 8.6127>}]
 Dyn_sleeve_l [{shapeType=Box, extentX=5.676, extentY=7.9676, extentZ=13.1011, rotation=<rot 90 0 0>}]
 Dyn_sleeve_r [{shapeType=Box, extentX=4.5111, extentY=7.0112, extentZ=13.851, rotation=<rot 90 0 0>}]
 calf_l [{shapeType=Box, extentX=7.5, extentY=7.5, extentZ=7.5}]
 calf_r [{shapeType=Box, extentX=7.5, extentY=7.5, extentZ=7.5}]
 clavicle_l [{shapeType=Box, extentX=4.8061, extentY=10.4407, extentZ=10.8052, center=<11.3117 -0.6081 -2.0702>, rotation=<rot 0 -171.1823 -67.6902>}]
 clavicle_r [{shapeType=Box, extentX=6.6853, extentY=9.4038, extentZ=10.698, center=<-10.1372 0.1976 2.1166>, rotation=<rot 0 -173.8363 -77.2904>}]
 pelvis [
  {shapeType=Capsule, length=10.1794, radius=9.2606, center=<-1.1326 -4.5029 0.0001>}
  {shapeType=Capsule, length=10.0579, radius=9.3821, center=<0 0 0.0001>}
 ]
 thigh_twist_01_l [{shapeType=Capsule, length=13.2535, radius=10.8109, center=<10.1896 -0.4415 1.1144>, rotation=<rot 90 0 0>}]
 thigh_twist_01_r [{shapeType=Capsule, length=16.0622, radius=9.1661, center=<-10.4624 0.1332 -0.706>, rotation=<rot 90 0 0>}]
 upperarm_l [{shapeType=Capsule, length=23, radius=7, center=<11.7449 0.0917 -1.963>, rotation=<rot 90 0 0>}]
 upperarm_r [{shapeType=Capsule, length=23, radius=7, center=<-11.7449 -0.0917 1.963>, rotation=<rot 90 0 0>}]
}
constraints [
 {bone1Name=Dyn_beltStrap02_l, bone2Name=Dyn_beltStrap01_l, swing1LimitDegrees=75, swing2LimitDegrees=75, twistLimitDegrees=45}
 {bone1Name=Dyn_beltStrap03_l, bone2Name=Dyn_beltStrap02_l, swing1LimitDegrees=75, swing2LimitDegrees=75, twistLimitDegrees=45}
 {bone1Name=Dyn_beltStrap01_r, bone2Name=Dyn_belt_mstr, swing1LimitDegrees=45, swing2LimitDegrees=45, twistLimitDegrees=45}
 {bone1Name=Dyn_beltStrap02_r, bone2Name=Dyn_beltStrap01_r, swing1LimitDegrees=45, swing2LimitDegrees=45, twistLimitDegrees=45}
 {bone1Name=Dyn_beltStrap03_r, bone2Name=Dyn_beltStrap02_r, swing1LimitDegrees=45, swing2LimitDegrees=45, twistLimitDegrees=45}
 {
  bone1Name=clavicle_l
  bone2Name=pelvis
  swing1LimitDegrees=45
  swing1Motion=Free
  swing2LimitDegrees=45
  swing2Motion=Free
  twistLimitDegrees=45
  twistMotion=Free
 }
 {
  bone1Name=clavicle_r
  bone2Name=pelvis
  swing1LimitDegrees=45
  swing1Motion=Free
  swing2LimitDegrees=45
  swing2Motion=Free
  twistLimitDegrees=45
  twistMotion=Free
 }
 {
  bone1Name=Dyn_sleeve_r
  bone2Name=clavicle_r
  swing1LimitDegrees=45
  swing1Motion=Locked
  swing2LimitDegrees=45
  swing2Motion=Locked
  twistLimitDegrees=45
  twistMotion=Locked
 }
 {
  bone1Name=Dyn_sleeve_l
  bone2Name=clavicle_l
  swing1LimitDegrees=45
  swing1Motion=Locked
  swing2LimitDegrees=45
  swing2Motion=Locked
  twistLimitDegrees=45
  twistMotion=Locked
 }
 {bone1Name=Dyn_pant_leg_r, bone2Name=calf_r, swing1LimitDegrees=15, swing2LimitDegrees=15, twistLimitDegrees=15}
 {bone1Name=Dyn_pant_leg_l, bone2Name=calf_l, swing1LimitDegrees=15, swing2LimitDegrees=15, twistLimitDegrees=15}
 {bone1Name=Dyn_skirt_e, bone2Name=Dyn_skirt_mstr, swing1LimitDegrees=45, swing2LimitDegrees=45, twistLimitDegrees=45}
 {bone1Name=Dyn_skirt_e_02, bone2Name=Dyn_skirt_e, swing1LimitDegrees=58.7143, swing2LimitDegrees=9, twistLimitDegrees=45}
 {bone1Name=Dyn_skirt_d, bone2Name=Dyn_skirt_mstr, swing1LimitDegrees=50.1429, swing2LimitDegrees=21, twistLimitDegrees=30.8571}
 {bone1Name=Dyn_skirt_d_02, bone2Name=Dyn_skirt_d, swing1LimitDegrees=70.7143, swing2LimitDegrees=17.1429, twistLimitDegrees=30.8571}
 {bone1Name=Dyn_skirt_c, bone2Name=Dyn_skirt_mstr, swing1LimitDegrees=51, swing2LimitDegrees=45, twistLimitDegrees=45}
 {bone1Name=Dyn_skirt_c_02, bone2Name=Dyn_skirt_c, swing1LimitDegrees=69.4286, swing2LimitDegrees=9, twistLimitDegrees=45}
 {bone1Name=Dyn_skirt_a_02, bone2Name=Dyn_skirt_a, swing1LimitDegrees=45, swing2LimitDegrees=5, swing2Motion=Locked, twistLimitDegrees=93}
 {bone1Name=Dyn_skirt_f_02, bone2Name=Dyn_skirt_f, swing1LimitDegrees=19.2857, swing2LimitDegrees=10, twistLimitDegrees=45}
 {bone1Name=Dyn_skirt_a, bone2Name=Dyn_skirt_f, swing1LimitDegrees=45, swing2LimitDegrees=45, swing2Motion=Locked, twistLimitDegrees=45}
 {
  bone1Name=Dyn_sleeve_adjust_l
  bone2Name=clavicle_l
  swing1LimitDegrees=45
  swing1Motion=Locked
  swing2LimitDegrees=45
  swing2Motion=Locked
  twistLimitDegrees=90
 }
 {
  bone1Name=Dyn_sleeve_adjust_r
  bone2Name=clavicle_r
  swing1LimitDegrees=45
  swing1Motion=Locked
  swing2LimitDegrees=45
  swing2Motion=Locked
  twistLimitDegrees=105
 }
 {
  bone1Name=Dyn_skirt_b
  bone2Name=Dyn_skirt_b_mstr
  swing1LimitDegrees=75
  swing2LimitDegrees=2.5
  swing2Motion=Locked
  twistLimitDegrees=11.4286
 }
 {
  bone1Name=Dyn_skirt_b_02
  bone2Name=Dyn_skirt_b
  swing1LimitDegrees=75
  swing2LimitDegrees=2.5
  swing2Motion=Locked
  twistLimitDegrees=15.4286
 }
 {bone1Name=Dyn_skirt_back, bone2Name=Dyn_skirt_mstr, swing1LimitDegrees=51, swing2LimitDegrees=45, twistLimitDegrees=45}
 {bone1Name=Dyn_skirt_back_02, bone2Name=Dyn_skirt_back, swing1LimitDegrees=69.4286, swing2LimitDegrees=9, twistLimitDegrees=45}
 {
  bone1Name=Dyn_skirt_a_02
  bone2Name=Dyn_skirt_f_02
  swing1LimitDegrees=45
  swing1Motion=Free
  swing2LimitDegrees=45
  twistLimitDegrees=45
  twistMotion=Free
 }
 {
  bone1Name=Dyn_skirt_a_mstr
  bone2Name=Dyn_skirt_a
  swing1LimitDegrees=45
  swing1Motion=Free
  swing2LimitDegrees=45
  twistLimitDegrees=87.8571
  twistMotion=Free
 }
 {bone1Name=Dyn_skirt_f_mstr, bone2Name=Dyn_skirt_f, swing1LimitDegrees=46.7143, swing2LimitDegrees=10.2857, twistLimitDegrees=45}
 {bone1Name=Dyn_beltStrap01_l, bone2Name=Dyn_beltStrap01_l_mstr, swing1LimitDegrees=75, swing2LimitDegrees=42.4286, twistLimitDegrees=27.8571}
]
