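# Physics asset for the sleeves of the male medium Cyber Karate body, covering left and right dynamic sleeve bones.
# Box bodies on clavicles and sleeves, with capsules on pelvis and upper arms, form the collision setup.
# Clavicles rotate freely against the pelvis; sleeve bones lock angularly to the clavicles, while adjust bones allow only twist, wider on the right.
# Upper arms and the right-only adjust master, whose box copies the right clavicle's, have no constraints.
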PhysicsAsset /BRCosmetics/Characters/Player/Male/Medium/Bodies/M_MED_Cyber_Karate/Meshes/M_MED_Cyber_Karate_PhysicsAsset_Sleeves
bodies {
 Dyn_sleeve_adjust_l [{shapeType=Box, extentX=30.5688, extentY=5.9, extentZ=15.4725, center=<0.3276 0 8.0449>}]
 Dyn_sleeve_adjust_r [{shapeType=Box, extentX=30.8736, extentY=4.8211, extentZ=16.0393, center=<0.13 0 8.6127>}]
 Dyn_sleeve_adjust_r_mstr [{shapeType=Box, extentX=6.6853, extentY=9.4038, extentZ=10.698, center=<-10.1372 0.1976 2.1166>, rotation=<rot 0 -173.8363 -77.2904>}]
 Dyn_sleeve_l [{shapeType=Box, extentX=5.676, extentY=7.9676, extentZ=13.1011, rotation=<rot 90 0 0>}]
 Dyn_sleeve_r [{shapeType=Box, extentX=4.5111, extentY=7.0112, extentZ=13.851, rotation=<rot 90 0 0>}]
 clavicle_l [{shapeType=Box, extentX=4.8061, extentY=10.4407, extentZ=10.8052, center=<11.3117 -0.6081 -2.0702>, rotation=<rot 0 -171.1823 -67.6902>}]
 clavicle_r [{shapeType=Box, extentX=6.6853, extentY=9.4038, extentZ=10.698, center=<-10.1372 0.1976 2.1166>, rotation=<rot 0 -173.8363 -77.2904>}]
 pelvis [{shapeType=Capsule, length=10.1794, radius=9.2606, center=<-1.1326 -4.5029 0.0001>}]
 upperarm_l [{shapeType=Capsule, length=23, radius=7, center=<11.7449 0.0917 -1.963>, rotation=<rot 90 0 0>}]
 upperarm_r [{shapeType=Capsule, length=23, radius=7, center=<-11.7449 -0.0917 1.963>, rotation=<rot 90 0 0>}]
}
constraints [
 {
  bone1Name=clavicle_l
  bone2Name=pelvis
  swing1LimitDegrees=45
  swing1Motion=Free
  swing2LimitDegrees=45
  swing2Motion=Free
  twistLimitDegrees=45
  twistMotion=Free
 }
 {
  bone1Name=clavicle_r
  bone2Name=pelvis
  swing1LimitDegrees=45
  swing1Motion=Free
  swing2LimitDegrees=45
  swing2Motion=Free
  twistLimitDegrees=45
  twistMotion=Free
 }
 {
  bone1Name=Dyn_sleeve_r
  bone2Name=clavicle_r
  swing1LimitDegrees=45
  swing1Motion=Locked
  swing2LimitDegrees=45
  swing2Motion=Locked
  twistLimitDegrees=45
  twistMotion=Locked
 }
 {
  bone1Name=Dyn_sleeve_l
  bone2Name=clavicle_l
  swing1LimitDegrees=45
  swing1Motion=Locked
  swing2LimitDegrees=45
  swing2Motion=Locked
  twistLimitDegrees=45
  twistMotion=Locked
 }
 {
  bone1Name=Dyn_sleeve_adjust_l
  bone2Name=clavicle_l
  swing1LimitDegrees=45
  swing1Motion=Locked
  swing2LimitDegrees=45
  swing2Motion=Locked
  twistLimitDegrees=90
 }
 {
  bone1Name=Dyn_sleeve_adjust_r
  bone2Name=clavicle_r
  swing1LimitDegrees=45
  swing1Motion=Locked
  swing2LimitDegrees=45
  swing2Motion=Locked
  twistLimitDegrees=142.7143
 }
]
