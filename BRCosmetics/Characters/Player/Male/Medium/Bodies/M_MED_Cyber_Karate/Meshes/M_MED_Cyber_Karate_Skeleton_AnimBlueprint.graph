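# Animation Blueprint for the Cyber Karate body cosmetic, copying the attached parent's pose and adding skirt, belt and sleeve secondary motion.
# Crouch events set CrouchingAlpha, which drives corrective skirt and belt bone offsets and crossfades standing and crouch RigidBody simulations via crouchINV.
# An always-on sleeve simulation follows, then constraints on sleeves and skirt, several at zero alpha.
# TargetingAlpha is set but unused, and one RigidBody node is disconnected.
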AnimBlueprint /BRCosmetics/Characters/Player/Male/Medium/Bodies/M_MED_Cyber_Karate/Meshes/M_MED_Cyber_Karate_Skeleton_AnimBlueprint
target_skeleton=&/M_MED_Cyber_Karate_Skeleton
graphs [
 AnimGraph {nodes [
   RigidBody() {
    bClampLinearTranslationLimitToRefPose=true
    bOverrideWorldGravity=true
    overridePhysicsAsset=&/M_MED_Cyber_Karate_02_PhysicsAsset
    simulationSpace=BaseBoneSpace
    alphaBoolBlend {blendOption=Linear}
    baseBoneRef {boneName=pelvis}
    componentAppliedLinearAccClamp=<3000 3000 3000>
    componentLinearAccScale=<0.3 0.3 0.3>
    componentLinearVelScale=<0.3 0.3 0.3>
    overrideWorldGravity {z=-980}
   }
   n1 = Get CrouchingAlpha() -> (CrouchingAlpha:double)
   n2 = "Transform (Modify) Bone - Bone: Dyn_skirt_mstr"(
    ComponentPose=Local To Component(LocalPose=Copy Pose From Mesh() {bUseAttachedParent=true})
    Alpha=@n1
    Translation="0,0,2"
    Rotation="0,0,-5"
    Scale="1,1,1"
   ) {
    rotationMode=BMM_Additive
    translationMode=BMM_Additive
    translationSpace=BCS_WorldSpace
    alphaBoolBlend {blendOption=Linear}
    alphaScaleBiasClamp {bClampResult=true, bInterpResult=true}
    boneToModify {boneName=Dyn_skirt_mstr}
    rotation {roll=-5}
    translation {z=2}
   }
   n3 = "Transform (Modify) Bone - Bone: Dyn_skirt_b_mstr"(ComponentPose=@n2, Alpha=@n1, Translation="2.764,2.892,5.629", Rotation="-9.040,51.830,-20.900", Scale="0.7,0.7,0.7") {
    rotationMode=BMM_Additive
    rotationSpace=BCS_BoneSpace
    scaleMode=BMM_Replace
    translationMode=BMM_Additive
    alphaBoolBlend {blendOption=Linear}
    alphaScaleBiasClamp {bClampResult=true, bInterpResult=true}
    boneToModify {boneName=Dyn_skirt_b_mstr}
    rotation=<rot -9.04 51.8303 -20.8999>
    scale=<0.7 0.7 0.7>
    translation=<2.764 2.892 5.629>
   }
   n4 = "Transform (Modify) Bone - Bone: Dyn_beltStrap01_l_mstr"(ComponentPose=@n3, Alpha=@n1, Translation="-4.209,2.292,-2.969", Rotation="-25.000,7.298,30.116", Scale="1,1,1") {
    rotationMode=BMM_Additive
    rotationSpace=BCS_BoneSpace
    translationMode=BMM_Additive
    translationSpace=BCS_BoneSpace
    alphaBoolBlend {blendOption=Linear}
    alphaScaleBiasClamp {bClampResult=true, bInterpResult=true, interpSpeedDecreasing=25, interpSpeedIncreasing=25}
    boneToModify {boneName=Dyn_beltStrap01_l_mstr}
    rotation=<rot -25 7.2981 30.116>
    translation=<-4.209 2.292 -2.969>
   }
   n5 = "Transform (Modify) Bone - Bone: Dyn_belt_mstr"(ComponentPose=@n4, Alpha=@n1, Translation="-5,2,0", Rotation="0.000,0.000,-4.869", Scale="1,1,1") {
    rotationMode=BMM_Additive
    translationMode=BMM_Additive
    translationSpace=BCS_BoneSpace
    alphaBoolBlend {blendOption=Linear}
    alphaScaleBiasClamp {bClampResult=true, bInterpResult=true, interpSpeedDecreasing=25, interpSpeedIncreasing=25}
    boneToModify {boneName=Dyn_belt_mstr}
    rotation {roll=-4.869}
    translation=<-5 2>
   }
   n6 = "Transform (Modify) Bone - Bone: Dyn_skirt_f_mstr"(ComponentPose=@n5, Alpha=@n1, Translation="-0.778,4.000,-2.000", Rotation="-1.349,26.109,27.389", Scale="0.7,0.7,0.7") {
    rotationMode=BMM_Additive
    rotationSpace=BCS_BoneSpace
    translationMode=BMM_Additive
    alphaBoolBlend {blendOption=Linear}
    alphaScaleBiasClamp {bClampResult=true, bInterpResult=true}
    boneToModify {boneName=Dyn_skirt_f_mstr}
    rotation=<rot -1.349 26.1089 27.3889>
    scale=<0.7 0.7 0.7>
    translation=<-0.778 4 -2>
   }
   n7 = "Transform (Modify) Bone - Bone: Dyn_skirt_a_mstr"!(ComponentPose=@n6, Alpha=@n1, Translation="3.883,3.000,5.367", Rotation="13.248,65.456,13.463") -> (Pose) {
    rotationMode=BMM_Additive
    rotationSpace=BCS_BoneSpace
    scaleMode=BMM_Replace
    translationMode=BMM_Additive
    alphaBoolBlend {blendOption=Linear}
    alphaScaleBiasClamp {bClampResult=true, bInterpResult=true}
    boneToModify {boneName=Dyn_skirt_a_mstr}
    rotation=<rot 13.248 65.4561 13.4631>
    scale=<0.8 0.8 0.8>
    translation=<3.883 3 5.367>
   }
   n8 = "Transform (Modify) Bone - Bone: Dyn_skirt_d"(ComponentPose=@n7, Alpha=@n1, Translation="0,0,0", Rotation="0,0,0", Scale="1,1,1") {
    rotationMode=BMM_Additive
    rotationSpace=BCS_BoneSpace
    scaleSpace=BCS_BoneSpace
    translationMode=BMM_Additive
    translationSpace=BCS_BoneSpace
    alphaBoolBlend {blendOption=Linear}
    alphaScaleBiasClamp {bClampResult=true, bInterpResult=true}
    boneToModify {boneName=Dyn_skirt_d}
   }
   n9 = RigidBody(ComponentPose=@n8, Alpha=Get crouchINV() -> (crouchINV:double)) {
    bClampLinearTranslationLimitToRefPose=true
    bOverrideWorldGravity=true
    overridePhysicsAsset=&/M_MED_Cyber_Karate_PhysicsAsset
    simulationSpace=BaseBoneSpace
    alphaBoolBlend {blendOption=Linear}
    baseBoneRef {boneName=pelvis}
    componentAppliedLinearAccClamp=<3000 3000 3000>
    componentLinearAccScale=<0.425 0.425 0.425>
    componentLinearVelScale=<0.425 0.425 0.425>
    overrideWorldGravity {z=-980}
   }
   n10 = RigidBody(ComponentPose=@n9, Alpha=Get CrouchingAlpha()) {
    bClampLinearTranslationLimitToRefPose=true
    bOverrideWorldGravity=true
    overridePhysicsAsset=&/M_MED_Cyber_Karate_CrouchPhysics
    simulationSpace=BaseBoneSpace
    alphaBoolBlend {blendOption=Linear}
    baseBoneRef {boneName=pelvis}
    componentAppliedLinearAccClamp=<3000 3000 3000>
    componentLinearAccScale=<0.425 0.425 0.425>
    componentLinearVelScale=<0.425 0.425 0.425>
    overrideWorldGravity {z=-980}
   }
   n11 = RigidBody(ComponentPose=@n10) {
    bClampLinearTranslationLimitToRefPose=true
    bOverrideWorldGravity=true
    overridePhysicsAsset=&/M_MED_Cyber_Karate_PhysicsAsset_Sleeves
    simulationSpace=BaseBoneSpace
    alphaBoolBlend {blendOption=Linear}
    baseBoneRef {boneName=spine_03}
    componentAppliedLinearAccClamp=<6000 6000 6000>
    componentLinearAccScale=<0.4 0.4 0.4>
    componentLinearVelScale=<0.4 0.4 0.4>
    overrideWorldGravity {z=-980}
   }
   n12 = "Constraint - Dyn_sleeve_adjust_l (P)"(ComponentPose=@n11, Alpha=0.75, ConstraintWeights_0=1) {
    alphaBoolBlend {blendOption=Linear}
    boneToModify {boneName=Dyn_sleeve_adjust_l}
    constraintSetup [{targetBone {boneName=upperarm_l}}]
    constraintWeights [1]
   }
   n13 = "Constraint - Dyn_sleeve_adjust_r (P)"(ComponentPose=@n12, Alpha=1, ConstraintWeights_0=1) {
    alphaBoolBlend {blendOption=Linear}
    boneToModify {boneName=Dyn_sleeve_adjust_r}
    constraintSetup [{targetBone {boneName=upperarm_r}}]
    constraintWeights [1]
   }
   n14 = "Constraint - Dyn_skirt_f (R)"(ComponentPose=@n13, Alpha=0, ConstraintWeights_0=1) {
    alphaBoolBlend {blendOption=Linear}
    alphaScaleBiasClamp {bClampResult=true, bInterpResult=true, clampMax=0.75}
    boneToModify {boneName=Dyn_skirt_f}
    constraintSetup [{transformType=Rotation, targetBone {boneName=Dyn_skirt_a}}]
    constraintWeights [1]
   }
   n15 = "Constraint - Dyn_skirt_f_02 (R)"(ComponentPose=@n14, Alpha=0, ConstraintWeights_0=1) {
    alphaBoolBlend {blendOption=Linear}
    alphaScaleBiasClamp {bInterpResult=true, clampMax=0.75}
    boneToModify {boneName=Dyn_skirt_f_02}
    constraintSetup [{transformType=Rotation, targetBone {boneName=Dyn_skirt_a_02}}]
    constraintWeights [1]
   }
   n16 = "Constraint - Dyn_skirt_b_mstr (R)"(ComponentPose=@n15, Alpha=0, ConstraintWeights_0=1) {
    alphaBoolBlend {blendOption=Linear}
    alphaScaleBiasClamp {bInterpResult=true, clampMax=0.75}
    boneToModify {boneName=Dyn_skirt_b_mstr}
    constraintSetup [{transformType=Rotation, targetBone {boneName=thigh_twist_01_l}}]
    constraintWeights [1]
   }
   n17 = "Constraint - Dyn_skirt_f_02 (P)"(ComponentPose=@n16, Alpha=Get CrouchingAlpha(), ConstraintWeights_0=1) {
    alphaBoolBlend {blendOption=Linear}
    alphaScaleBiasClamp {bClampResult=true, bInterpResult=true, bMapRange=true, clampMax=0.75}
    boneToModify {boneName=Dyn_skirt_f_02}
    constraintSetup [{targetBone {boneName=Dyn_skirt_a_02}}]
    constraintWeights [1]
   }
   n18 = "Constraint - Dyn_skirt_f (P)"(ComponentPose=@n17, Alpha=0, ConstraintWeights_0=1) {
    alphaBoolBlend {blendOption=Linear}
    alphaScaleBiasClamp {bClampResult=true, bInterpResult=true, bMapRange=true, clampMax=0.75}
    boneToModify {boneName=Dyn_skirt_f}
    constraintSetup [{targetBone {boneName=Dyn_skirt_a_02}}]
    constraintWeights [1]
   }
   n19 = "Transform (Modify) Bone - Bone: Dyn_sleeve_adjust_r_mstr"!(ComponentPose=@n18, Alpha=0, Translation="-0.75,-3.5,-1") -> (Pose) {
    translationMode=BMM_Additive
    translationSpace=BCS_BoneSpace
    alphaBoolBlend {blendOption=Linear}
    alphaScaleBiasClamp {bClampResult=true, bInterpResult=true}
    boneToModify {boneName=Dyn_sleeve_adjust_r_mstr}
    translation=<-0.75 -3.5 -1>
   }
   Output Pose(Result=Component To Local(ComponentPose=@n19)) {layerGroup=None}
   Get CrouchingAlpha()
   Get TargetingAlpha() -> (TargetingAlpha:double)
  ]}
 EventGraph {nodes [
   Event BlueprintUpdateAnimation()
   > Set crouchINV(crouchINV:double=float - float(B=Get CrouchingAlpha()) {is=title K2Node_CallFunction}) -> (Output_Get:double)
   Event OnBeginCrouching()
   > Set CrouchingAlpha(CrouchingAlpha:double=1) -> (Output_Get:double)
   Event OnEndCrouching()
   > Set CrouchingAlpha(CrouchingAlpha=0)
   Event OnBeginTargeting() {is=title K2Node_Event}
   > Set TargetingAlpha(TargetingAlpha:double=1) -> (Output_Get:double)
   Event OnEndTargeting() {is=title K2Node_Event}
   > Set TargetingAlpha(TargetingAlpha=0)
   TryGetPawnOwner()
  ]}
]
variables [
 CrouchingAlpha {type=number}
 crouchINV {type=number}
 TargetingAlpha {type=number, default=1}
]
