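# Physics asset for the face accessory of the male medium Cyber Karate player cosmetic, covering a pair of dangling ribbons.
# Left and right five-bone chains hang from Dyn_ribbon_mstr, using boxes near the root and capsules further out, with limited twist and swing that loosens toward the tips.
# Head, clavicle and upper-spine bodies supply body collision shapes.
# Free angular constraints link the master, head and spine bodies, while the clavicles have no constraints.
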PhysicsAsset /BRCosmetics/Characters/Player/Male/Medium/Bodies/M_MED_Cyber_Karate/Meshes/Parts/M_MED_Cyber_Karate_FaceAcc_PhysicsAsset
bodies {
 Dyn_ribbon_01_l [{shapeType=Box, extentX=1.4975, extentY=2.7303, extentZ=8.3055, center=<4.1634 0 0.2212>, rotation=<rot 83.8935 -89.9993 -89.9993>}]
 Dyn_ribbon_01_r [{shapeType=Box, extentX=1.5107, extentY=2.8653, extentZ=7, center=<3.4335 -0.3441 0.3053>, rotation=<rot 82.2953 -180 -169.436>}]
 Dyn_ribbon_02_l [{shapeType=Box, extentX=1.5376, extentY=2.7541, extentZ=7, center=<4 0 0>, rotation=<rot 90 0 0>}]
 Dyn_ribbon_02_r [{shapeType=Box, extentX=1.2238, extentY=3.1512, extentZ=8.2004, center=<3.9085 -0.4947 0>, rotation=<rot 90 0 10.475>}]
 Dyn_ribbon_03_l [{shapeType=Capsule, length=6.3013, radius=1.7244, center=<3.7907 -0.0061 -0.0001>, rotation=<rot 90 0 0>}]
 Dyn_ribbon_03_r [{shapeType=Capsule, length=5.1101, radius=2.3052, center=<4.02 0.0717 0>, rotation=<rot 90 0 0>}]
 Dyn_ribbon_04_l [{shapeType=Capsule, length=5.5447, radius=1.7927, center=<4 -0.0023 0>, rotation=<rot 90 0 0>}]
 Dyn_ribbon_04_r [{shapeType=Capsule, length=7.1621, radius=2.338, center=<4.5288 0.1395 -0.0001>, rotation=<rot 90 0 0>}]
 Dyn_ribbon_05_l [{shapeType=Capsule, length=8.0965, radius=2.341, center=<5.7039 -0.2027 0.0001>, rotation=<rot 90 0 0>}]
 Dyn_ribbon_05_r [{shapeType=Capsule, length=7.3501, radius=2.4538, center=<5.5459 0.4822 0>, rotation=<rot 90 0 0>}]
 Dyn_ribbon_mstr [{shapeType=Box, extentX=6, extentY=1, extentZ=7.5}]
 clavicle_l [{shapeType=Sphere, radius=15, center=<15.7004 2.8347 -4.229>}]
 clavicle_r [{shapeType=Sphere, radius=15, center=<-17.0739 -0.4862 6.6619>}]
 head [{shapeType=Capsule, length=16.3867, radius=10.1761, center=<0 3.8293 0.0006>, rotation=<rot 90 0 0>}]
 spine_04 [{shapeType=Capsule, length=10.5, radius=13.75, center=<-2.6616 2.7376 0>}]
 spine_05 [
  {shapeType=Sphere, radius=13.75, center=<0 -1.0924 18.3686>}
  {shapeType=Sphere, radius=13, center=<1.3026 -2.0855 -18.7389>}
  {shapeType=Capsule, length=28.5, radius=15, center=<-1.033 1.3943 0.0001>}
 ]
}
constraints [
 {bone1Name=Dyn_ribbon_01_l, bone2Name=Dyn_ribbon_mstr, swing1LimitDegrees=20, swing2LimitDegrees=20, twistLimitDegrees=5}
 {bone1Name=Dyn_ribbon_02_l, bone2Name=Dyn_ribbon_01_l, swing1LimitDegrees=20, swing2LimitDegrees=20, twistLimitDegrees=45}
 {bone1Name=Dyn_ribbon_03_l, bone2Name=Dyn_ribbon_02_l, swing1LimitDegrees=20, swing2LimitDegrees=20, twistLimitDegrees=5}
 {bone1Name=Dyn_ribbon_04_l, bone2Name=Dyn_ribbon_03_l, swing1LimitDegrees=20, swing2LimitDegrees=40, twistLimitDegrees=5}
 {bone1Name=Dyn_ribbon_05_l, bone2Name=Dyn_ribbon_04_l, swing1LimitDegrees=20, swing2LimitDegrees=60, twistLimitDegrees=5}
 {bone1Name=Dyn_ribbon_01_r, bone2Name=Dyn_ribbon_mstr, swing1LimitDegrees=10, swing2LimitDegrees=20, twistLimitDegrees=5}
 {bone1Name=Dyn_ribbon_02_r, bone2Name=Dyn_ribbon_01_r, swing1LimitDegrees=10, swing2LimitDegrees=20, twistLimitDegrees=5}
 {bone1Name=Dyn_ribbon_03_r, bone2Name=Dyn_ribbon_02_r, swing1LimitDegrees=10, swing2LimitDegrees=20, twistLimitDegrees=5}
 {bone1Name=Dyn_ribbon_04_r, bone2Name=Dyn_ribbon_03_r, swing1LimitDegrees=10, swing2LimitDegrees=40, twistLimitDegrees=5}
 {bone1Name=Dyn_ribbon_05_r, bone2Name=Dyn_ribbon_04_r, swing1LimitDegrees=20, swing2LimitDegrees=60, twistLimitDegrees=5}
 {
  bone1Name=head
  bone2Name=spine_05
  swing1LimitDegrees=45
  swing1Motion=Free
  swing2LimitDegrees=45
  swing2Motion=Free
  twistLimitDegrees=45
  twistMotion=Free
 }
 {
  bone1Name=Dyn_ribbon_mstr
  bone2Name=spine_05
  swing1LimitDegrees=45
  swing1Motion=Free
  swing2LimitDegrees=45
  swing2Motion=Free
  twistLimitDegrees=45
  twistMotion=Free
 }
 {
  bone1Name=spine_05
  bone2Name=spine_04
  swing1LimitDegrees=45
  swing1Motion=Free
  swing2LimitDegrees=45
  swing2Motion=Free
  twistLimitDegrees=45
  twistMotion=Free
 }
]
