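# Post-copy animation blueprint for the male Cyber Karate face accessory, adding physics-driven motion to the cosmetic part.
# Copies the attached parent's pose into a cache feeding two RigidBody simulations.
# IfJumping blends between a root-based sim with stronger gravity and motion response (true) and a damped spine_01-based one.
# Updates and JumpEntered notifies set IfJumping from upward velocity, OR'd with itself, so it latches once true.
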AnimBlueprint /BRCosmetics/Characters/Player/Male/Medium/Bodies/M_MED_Cyber_Karate/Meshes/Parts/M_MED_Cyber_Karate_FaceAcc_Skeleton_AnimBlueprint
target_skeleton=&/M_MED_Cyber_Karate_FaceAcc_Skeleton
graphs [
 AnimGraph {nodes [
   n1 = RigidBody(ComponentPose=Local To Component(LocalPose=Use cached pose 'CopyPose'())) {
    bClampLinearTranslationLimitToRefPose=true
    bOverrideWorldGravity=true
    overridePhysicsAsset=&/M_MED_Cyber_Karate_FaceAcc_PhysicsAsset
    alphaBoolBlend {blendOption=Linear}
    alphaScaleBiasClamp {bInterpResult=true, interpSpeedDecreasing=4, interpSpeedIncreasing=6}
    baseBoneRef {boneName=spine_01}
    componentAppliedLinearAccClamp=<2000 2000 2000>
    componentLinearAccScale=<0.15 0.15 0.15>
    componentLinearVelScale=<0.15 0.15 0.15>
    overrideWorldGravity {z=-980}
   }
   n2 = RigidBody(ComponentPose=Local To Component(LocalPose=Use cached pose 'CopyPose'())) {
    bClampLinearTranslationLimitToRefPose=true
    bOverrideWorldGravity=true
    overridePhysicsAsset=&M_MED_Cyber_Karate_FaceAcc_PhysicsAsset
    alphaBoolBlend {blendOption=Linear}
    alphaScaleBiasClamp {bInterpResult=true, interpSpeedDecreasing=4, interpSpeedIncreasing=6}
    baseBoneRef {boneName=root}
    componentAppliedLinearAccClamp=<6000 6000 6000>
    componentLinearAccScale=<0.5 0.5 0.5>
    componentLinearVelScale=<0.5 0.5 0.5>
    overrideWorldGravity {z=-1150}
   }
   Output Pose(Result=Blend Poses by bool(
     BlendPose_0=Component To Local(ComponentPose=@n2)
     BlendPose_1=Component To Local(ComponentPose=@n1)
     BlendTime_1=0.4
     bActiveValue=Get IfJumping() -> (IfJumping:bool)
    ) {blendType=Linear}) {layerGroup=None}
   Save cached pose 'CopyPose'(Pose=Copy Pose From Mesh() {bUseAttachedParent=true}) {cacheName=CopyPose}
  ]}
 EventGraph {nodes [
   n1 = Event BlueprintUpdateAnimation()
   n2 = Get PawnVelocityZ()
   n3 = Branch(
    execute=[@n1.then, @n4.then]
    Condition=OR Boolean!(A="float >= float"(A=@n2, B=1) {is=title K2Node_CallFunction}, B=false, C:bool=Get IfJumping()) -> (ReturnValue)
   )
   > Set IfJumping(IfJumping:bool=true) -> (Output_Get:bool)
   Set IfJumping(execute=@n3.else, IfJumping=false)
   n4 = AnimNotify_JumpEntered() {is=title K2Node_Event}
   TryGetPawnOwner()
   Get bIsFalling()
  ]}
]
variables [IfJumping {type=boolean}]
